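; A Season 16 battle pass quest actor for a crystal message that changes the look of a nearby small tower for the player.
; On BeginPlay a sphere overlap finds and caches an Apollo_SmallTower_01 actor, printing a debug message if none is found.
; When state changes for a player, non-authority clients swap the tower mesh's material to an emissive override.
; A testing reset restores the default material after a cooldown delay.
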(event EventBeginPlay
  (|CallParentFunction)
  (bind (_outactors _returnvalue) (Collision|SphereOverlapActors (Transformation|GetActorLocation) 1e+02 (Utilities|Array|MakeArray "ObjectTypeQuery1") "/Game/Environments/Apollo/Sets/Tower/Apollo_SmallTower_01.Apollo_SmallTower_01_C"))
  (if _returnvalue
    (bind _asapollo_small_tower_01 (Utilities|Casting|CastToApollo_SmallTower_01 (Utilities|Array|Get(acopy) _outactors))
      (:then
        (Variables|Default|SetTowerActor _asapollo_small_tower_01))
      (:CastFailed
        (Development|PrintString (Utilities|String|Append (Utilities|GetDisplayName self) " couldn't cast overlapped actor \"" (Utilities|GetDisplayName (Utilities|Array|Get(acopy) _outactors)) "\" to \"Apollo_SmallTower_01\"."))))
    (else
      (Development|PrintString (Utilities|String|Append (Utilities|GetDisplayName self) " didn't overlap actor of type: \"Apollo_SmallTower_01\"")))))

(event EventResetforVisualTesting
  (params (PlayerController :type Fort%20Player%20Controller%20Object%20Reference))
  (|CallParentFunction PlayerController)
  (Utilities|FlowControl|Delay (Variables|Design|QuestSetup|GetTestingCooldown))
  (Networking|SwitchHasAuthority
    (:Remote
      (Utilities|IsValid (Variables|Default|GetTowerActor)
        (:Is%20Valid
          (Rendering|Material|SetMaterial (Variables|/Script/Engine.StaticMeshActor|GetStaticMeshComponent (Variables|Default|GetTowerActor)) 0 (Variables|Default|GetEmissiveMIDDefault)))
        (:Is%20Not%20Valid
          (Development|PrintString (Utilities|String|Append (Utilities|GetDisplayName self) " attempted to set material of its tower actor to its defaults, but the actor was invalid.")))))))

(event EventChangeStateForPlayer
  (params (FPC :type Fort%20Player%20Controller%20Object%20Reference))
  (|CallParentFunction FPC)
  (Networking|SwitchHasAuthority
    (:Remote
      (Utilities|IsValid (Variables|Default|GetTowerActor)
        (:Is%20Valid
          (Rendering|Material|SetMaterial (Variables|/Script/Engine.StaticMeshActor|GetStaticMeshComponent (Variables|Default|GetTowerActor)) 0 (Variables|Default|GetEmissiveMIDOverride)))
        (:Is%20Not%20Valid
          (Development|PrintString (Utilities|String|Append (Utilities|GetDisplayName self) " attempted to set material of its tower actor to an emissive override, but the actor was invalid.")))))))

(fn ConstructionScript
  (|CallParentFunction))
