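; Season 16 Battle Pass quest actor for a tower glyph that switches a referenced tower mesh to an emissive look.
; The SpawnCosmeticActor override calls the parent, then sets material slot 0 on TowerActor's static mesh to EmissiveMIDOverride.
; The visual-testing reset calls the parent, waits the testing cooldown, then restores EmissiveMIDDefault.
; Both material swaps run only on the Remote branch of the authority switch, and all other quest logic is inherited.
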(event EventBeginPlay
  (|CallParentFunction) ; This node is disabled and will not be called.\nDrag off pins to build functionality.
)

(event Collision|EventActorBeginOverlap
  (params (OtherActor :type Actor%20Object%20Reference))
  (|CallParentFunction OtherActor) ; This node is disabled and will not be called.\nDrag off pins to build functionality.
)

(event EventTick
  (params (DeltaSeconds :type Float%20%28single-precision%29))
  (|CallParentFunction DeltaSeconds) ; This node is disabled and will not be called.\nDrag off pins to build functionality.
)

(event EventResetforVisualTesting
  (params (PlayerController :type Fort%20Player%20Controller%20Object%20Reference))
  (|CallParentFunction PlayerController)
  (Utilities|FlowControl|Delay (Variables|Design|QuestSetup|GetTestingCooldown))
  (Networking|SwitchHasAuthority
    (:Remote
      (Rendering|Material|SetMaterial (Variables|/Script/Engine.StaticMeshActor|GetStaticMeshComponent (Variables|Default|GetTowerActor)) 0 (Variables|Default|GetEmissiveMIDDefault)))))

(fn SpawnCosmeticActor
  (|CallParentFunction)
  (Networking|SwitchHasAuthority
    (:Remote
      (Rendering|Material|SetMaterial (Variables|/Script/Engine.StaticMeshActor|GetStaticMeshComponent (Variables|Default|GetTowerActor)) 0 (Variables|Default|GetEmissiveMIDOverride)))))

(fn ConstructionScript
  (|CallParentFunction))
